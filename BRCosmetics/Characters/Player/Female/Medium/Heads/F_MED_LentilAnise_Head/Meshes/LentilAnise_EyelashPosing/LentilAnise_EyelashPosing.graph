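# Level Sequence for the LentilAnise female medium head cosmetic, named for eyelash posing and apparently an authoring scene for that pose.
# A spawn track creates a FortPlayerMannequin, which a transform track positions and a Fortnite_Biped_CtrlRig track poses through channel defaults.
# Its child Head skeletal mesh component carries a layered Fortnite_ExtraJoint_CtrlRig track, so extra-joint adjustments sit on top of the underlying head pose.
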LevelSequence /BRCosmetics/Characters/Player/Female/Medium/Heads/F_MED_LentilAnise_Head/Meshes/LentilAnise_EyelashPosing/LentilAnise_EyelashPosing
bindings [
 Head {
  id=#1
  parent_id=#2
  possessed_class=SkeletalMeshComponentBudgeted
  tracks [{
    type=MovieSceneControlRigParameterTrack
    display_name="Fortnite_ExtraJoint_CtrlRig (Layered)"
    sections [{type=MovieSceneControlRigParameterSection}]
   }]
 }
 mannequin {
  id=#2
  possessed_class=FortPlayerMannequin
  tracks [
   {type=MovieSceneSpawnTrack, sections [{type=MovieSceneSpawnSection, channels [None {type=MovieSceneScriptingBoolChannel, default=true}]}]}
   {
    type=MovieScene3DTransformTrack
    sections [{
      type=MovieScene3DTransformSection
      channels [
       Location.X {type=MovieSceneScriptingDoubleChannel, default=0}
       Location.Y {type=MovieSceneScriptingDoubleChannel, default=0}
       Location.Z {type=MovieSceneScriptingDoubleChannel, default=1920.7656}
       Rotation.X {type=MovieSceneScriptingDoubleChannel, default=0}
       Rotation.Y {type=MovieSceneScriptingDoubleChannel, default=0}
       Rotation.Z {type=MovieSceneScriptingDoubleChannel, default=0}
       Scale.X {type=MovieSceneScriptingDoubleChannel, default=1}
       Scale.Y {type=MovieSceneScriptingDoubleChannel, default=1}
       Scale.Z {type=MovieSceneScriptingDoubleChannel, default=1}
      ]
     }]
   }
   {
    type=MovieSceneControlRigParameterTrack
    display_name=Fortnite_Biped_CtrlRig
    sections [{
      type=MovieSceneControlRigParameterSection
      channels [
       leg_l_stretch_switch {type=MovieSceneScriptingBoolChannel, default=false}
       leg_r_stretch_switch {type=MovieSceneScriptingBoolChannel, default=false}
       arm_l_fk_local {type=MovieSceneScriptingBoolChannel, default=true}
       Auto Interpolate_3 {type=MovieSceneScriptingBoolChannel, default=false}
       Auto Interpolate {type=MovieSceneScriptingBoolChannel, default=false}
       arm_l_stretch_switch {type=MovieSceneScriptingBoolChannel, default=false}
       Keep Aligned {type=MovieSceneScriptingBoolChannel, default=false}
       arm_r_fk_local {type=MovieSceneScriptingBoolChannel, default=true}
       Auto Interpolate_4 {type=MovieSceneScriptingBoolChannel, default=false}
       Auto Interpolate_2 {type=MovieSceneScriptingBoolChannel, default=false}
       arm_r_stretch_switch {type=MovieSceneScriptingBoolChannel, default=false}
       Keep Aligned_2 {type=MovieSceneScriptingBoolChannel, default=false}
       head_local {type=MovieSceneScriptingBoolChannel, default=true}
       Stretch {type=MovieSceneScriptingBoolChannel, default=false}
       head_stretch {type=MovieSceneScriptingBoolChannel, default=false}
       spine_stretch_switch {type=MovieSceneScriptingBoolChannel, default=false}
       arm_l_fk_ik_switch {type=MovieSceneScriptingBoolChannel, default=false}
       leg_l_fk_ik_switch {type=MovieSceneScriptingBoolChannel, default=true}
       arm_r_fk_ik_switch {type=MovieSceneScriptingBoolChannel, default=false}
       leg_r_fk_ik_switch {type=MovieSceneScriptingBoolChannel, default=true}
       ShowBodyControls {type=MovieSceneScriptingBoolChannel, default=true}
       spine_fk_ik_switch {type=MovieSceneScriptingBoolChannel, default=false}
       neck_fk_ik_switch {type=MovieSceneScriptingBoolChannel, default=false}
       space {type=MovieSceneScriptingByteChannel, default=1}
       global_ctrl.Location.X {type=MovieSceneScriptingFloatChannel, default=0}
       global_ctrl.Location.Y {type=MovieSceneScriptingFloatChannel, default=0}
       global_ctrl.Location.Z {type=MovieSceneScriptingFloatChannel, default=0}
       global_ctrl.Rotation.X {type=MovieSceneScriptingFloatChannel, default=0}
       global_ctrl.Rotation.Y {type=MovieSceneScriptingFloatChannel, default=0}
       global_ctrl.Rotation.Z {type=MovieSceneScriptingFloatChannel, default=0}
       root_ctrl.Location.X {type=MovieSceneScriptingFloatChannel, default=0}
       root_ctrl.Location.Y {type=MovieSceneScriptingFloatChannel, default=0}
       root_ctrl.Location.Z {type=MovieSceneScriptingFloatChannel, default=0}
       root_ctrl.Rotation.X {type=MovieSceneScriptingFloatChannel, default=0}
       root_ctrl.Rotation.Y {type=MovieSceneScriptingFloatChannel, default=0}
       root_ctrl.Rotation.Z {type=MovieSceneScriptingFloatChannel, default=0}
       root_ctrl.Scale.X {type=MovieSceneScriptingFloatChannel, default=1}
       root_ctrl.Scale.Y {type=MovieSceneScriptingFloatChannel, default=1}
       root_ctrl.Scale.Z {type=MovieSceneScriptingFloatChannel, default=1}
       body_offset_ctrl.Location.X {type=MovieSceneScriptingFloatChannel, default=0}
       body_offset_ctrl.Location.Y {type=MovieSceneScriptingFloatChannel, default=0}
       body_offset_ctrl.Location.Z {type=MovieSceneScriptingFloatChannel, default=0}
       body_offset_ctrl.Rotation.X {type=MovieSceneScriptingFloatChannel, default=0}
       body_offset_ctrl.Rotation.Y {type=MovieSceneScriptingFloatChannel, default=0}
       body_offset_ctrl.Rotation.Z {type=MovieSceneScriptingFloatChannel, default=0}
       body_offset_ctrl.Scale.X {type=MovieSceneScriptingFloatChannel, default=1}
       body_offset_ctrl.Scale.Y {type=MovieSceneScriptingFloatChannel, default=1}
       body_offset_ctrl.Scale.Z {type=MovieSceneScriptingFloatChannel, default=1}
       thigh_l_fk_ctrl.Location.X {type=MovieSceneScriptingFloatChannel, default=0}
       thigh_l_fk_ctrl.Location.Y {type=MovieSceneScriptingFloatChannel, default=0}
       thigh_l_fk_ctrl.Location.Z {type=MovieSceneScriptingFloatChannel, default=0}
       thigh_l_fk_ctrl.Rotation.X {type=MovieSceneScriptingFloatChannel, default=1.216}
       thigh_l_fk_ctrl.Rotation.Y {type=MovieSceneScriptingFloatChannel, default=-1.2128}
       thigh_l_fk_ctrl.Rotation.Z {type=MovieSceneScriptingFloatChannel, default=-0.1664}
       thigh_l_fk_ctrl.Scale.X {type=MovieSceneScriptingFloatChannel, default=1}
       thigh_l_fk_ctrl.Scale.Y {type=MovieSceneScriptingFloatChannel, default=1}
       thigh_l_fk_ctrl.Scale.Z {type=MovieSceneScriptingFloatChannel, default=1}
       calf_l_fk_ctrl.Location.X {type=MovieSceneScriptingFloatChannel, default=0}
       calf_l_fk_ctrl.Location.Y {type=MovieSceneScriptingFloatChannel, default=0}
       calf_l_fk_ctrl.Location.Z {type=MovieSceneScriptingFloatChannel, default=0}
       calf_l_fk_ctrl.Rotation.X {type=MovieSceneScriptingFloatChannel, default=0.0543}
       calf_l_fk_ctrl.Rotation.Y {type=MovieSceneScriptingFloatChannel, default=1.2238}
       calf_l_fk_ctrl.Rotation.Z {type=MovieSceneScriptingFloatChannel, default=0.3732}
       calf_l_fk_ctrl.Scale.X {type=MovieSceneScriptingFloatChannel, default=1}
       calf_l_fk_ctrl.Scale.Y {type=MovieSceneScriptingFloatChannel, default=1}
       calf_l_fk_ctrl.Scale.Z {type=MovieSceneScriptingFloatChannel, default=1}
       foot_l_fk_ctrl.Location.X {type=MovieSceneScriptingFloatChannel, default=0}
       foot_l_fk_ctrl.Location.Y {type=MovieSceneScriptingFloatChannel, default=0.0398}
       foot_l_fk_ctrl.Location.Z {type=MovieSceneScriptingFloatChannel, default=-0.8885}
       foot_l_fk_ctrl.Rotation.X {type=MovieSceneScriptingFloatChannel, default=-1.1647}
       foot_l_fk_ctrl.Rotation.Y {type=MovieSceneScriptingFloatChannel, default=0.0792}
       foot_l_fk_ctrl.Rotation.Z {type=MovieSceneScriptingFloatChannel, default=-0.2074}
       foot_l_fk_ctrl.Scale.X {type=MovieSceneScriptingFloatChannel, default=1}
       foot_l_fk_ctrl.Scale.Y {type=MovieSceneScriptingFloatChannel, default=1}
       foot_l_fk_ctrl.Scale.Z {type=MovieSceneScriptingFloatChannel, default=1}
       ball_l_fk_ctrl.Location.X {type=MovieSceneScriptingFloatChannel, default=0}
       ball_l_fk_ctrl.Location.Y {type=MovieSceneScriptingFloatChannel, default=0}
       ball_l_fk_ctrl.Location.Z {type=MovieSceneScriptingFloatChannel, default=0}
       ball_l_fk_ctrl.Rotation.X {type=MovieSceneScriptingFloatChannel, default=0}
       ball_l_fk_ctrl.Rotation.Y {type=MovieSceneScriptingFloatChannel, default=0}
       ball_l_fk_ctrl.Rotation.Z {type=MovieSceneScriptingFloatChannel, default=0}
       ball_l_fk_ctrl.Scale.X {type=MovieSceneScriptingFloatChannel, default=1}
       ball_l_fk_ctrl.Scale.Y {type=MovieSceneScriptingFloatChannel, default=1}
       ball_l_fk_ctrl.Scale.Z {type=MovieSceneScriptingFloatChannel, default=1}
       foot_l_ik_ctrl.Location.X {type=MovieSceneScriptingFloatChannel, default=0}
       foot_l_ik_ctrl.Location.Y {type=MovieSceneScriptingFloatChannel, default=0}
       foot_l_ik_ctrl.Location.Z {type=MovieSceneScriptingFloatChannel, default=0}
       foot_l_ik_ctrl.Rotation.X {type=MovieSceneScriptingFloatChannel, default=0}
       foot_l_ik_ctrl.Rotation.Y {type=MovieSceneScriptingFloatChannel, default=0}
       foot_l_ik_ctrl.Rotation.Z {type=MovieSceneScriptingFloatChannel, default=0}
       foot_l_ik_ctrl.Scale.X {type=MovieSceneScriptingFloatChannel, default=1}
       foot_l_ik_ctrl.Scale.Y {type=MovieSceneScriptingFloatChannel, default=1}
       foot_l_ik_ctrl.Scale.Z {type=MovieSceneScriptingFloatChannel, default=1}
       ball_l_ik_ctrl.Location.X {type=MovieSceneScriptingFloatChannel, default=0}
       ball_l_ik_ctrl.Location.Y {type=MovieSceneScriptingFloatChannel, default=0}
       ball_l_ik_ctrl.Location.Z {type=MovieSceneScriptingFloatChannel, default=0}
       ball_l_ik_ctrl.Rotation.X {type=MovieSceneScriptingFloatChannel, default=0}
       ball_l_ik_ctrl.Rotation.Y {type=MovieSceneScriptingFloatChannel, default=0}
       ball_l_ik_ctrl.Rotation.Z {type=MovieSceneScriptingFloatChannel, default=0}
       ball_l_ik_ctrl.Scale.X {type=MovieSceneScriptingFloatChannel, default=1}
       ball_l_ik_ctrl.Scale.Y {type=MovieSceneScriptingFloatChannel, default=1}
       ball_l_ik_ctrl.Scale.Z {type=MovieSceneScriptingFloatChannel, default=1}
       foot_bk1_l_ctrl.Location.X {type=MovieSceneScriptingFloatChannel, default=0}
       foot_bk1_l_ctrl.Location.Y {type=MovieSceneScriptingFloatChannel, default=0}
       foot_bk1_l_ctrl.Location.Z {type=MovieSceneScriptingFloatChannel, default=0}
       foot_bk1_l_ctrl.Rotation.X {type=MovieSceneScriptingFloatChannel, default=0}
       foot_bk1_l_ctrl.Rotation.Y {type=MovieSceneScriptingFloatChannel, default=0}
       foot_bk1_l_ctrl.Rotation.Z {type=MovieSceneScriptingFloatChannel, default=0}
       foot_bk1_l_ctrl.Scale.X {type=MovieSceneScriptingFloatChannel, default=1}
       foot_bk1_l_ctrl.Scale.Y {type=MovieSceneScriptingFloatChannel, default=1}
       foot_bk1_l_ctrl.Scale.Z {type=MovieSceneScriptingFloatChannel, default=1}
       foot_roll_l_ctrl.Location.X {type=MovieSceneScriptingFloatChannel, default=0}
       foot_roll_l_ctrl.Location.Y {type=MovieSceneScriptingFloatChannel, default=0}
       foot_roll_l_ctrl.Location.Z {type=MovieSceneScriptingFloatChannel, default=0}
       foot_roll_l_ctrl.Rotation.X {type=MovieSceneScriptingFloatChannel, default=0}
       foot_roll_l_ctrl.Rotation.Y {type=MovieSceneScriptingFloatChannel, default=0}
       foot_roll_l_ctrl.Rotation.Z {type=MovieSceneScriptingFloatChannel, default=0}
       foot_roll_l_ctrl.Scale.X {type=MovieSceneScriptingFloatChannel, default=1}
       foot_roll_l_ctrl.Scale.Y {type=MovieSceneScriptingFloatChannel, default=1}
       foot_roll_l_ctrl.Scale.Z {type=MovieSceneScriptingFloatChannel, default=1}
       roll_blend_l_ctrl {type=MovieSceneScriptingFloatChannel, default=1}
       tip_l_ctrl.Location.X {type=MovieSceneScriptingFloatChannel, default=0}
       tip_l_ctrl.Location.Y {type=MovieSceneScriptingFloatChannel, default=0}
       tip_l_ctrl.Location.Z {type=MovieSceneScriptingFloatChannel, default=0}
       tip_l_ctrl.Rotation.X {type=MovieSceneScriptingFloatChannel, default=0}
       tip_l_ctrl.Rotation.Y {type=MovieSceneScriptingFloatChannel, default=0}
       tip_l_ctrl.Rotation.Z {type=MovieSceneScriptingFloatChannel, default=0}
       tip_l_ctrl.Scale.X {type=MovieSceneScriptingFloatChannel, default=1}
       tip_l_ctrl.Scale.Y {type=MovieSceneScriptingFloatChannel, default=1}
       tip_l_ctrl.Scale.Z {type=MovieSceneScriptingFloatChannel, default=1}
       heel_l_ctrl.Location.X {type=MovieSceneScriptingFloatChannel, default=0}
       heel_l_ctrl.Location.Y {type=MovieSceneScriptingFloatChannel, default=0}
       heel_l_ctrl.Location.Z {type=MovieSceneScriptingFloatChannel, default=0}
       heel_l_ctrl.Rotation.X {type=MovieSceneScriptingFloatChannel, default=0}
       heel_l_ctrl.Rotation.Y {type=MovieSceneScriptingFloatChannel, default=0}
       heel_l_ctrl.Rotation.Z {type=MovieSceneScriptingFloatChannel, default=0}
       heel_l_ctrl.Scale.X {type=MovieSceneScriptingFloatChannel, default=1}
       heel_l_ctrl.Scale.Y {type=MovieSceneScriptingFloatChannel, default=1}
       heel_l_ctrl.Scale.Z {type=MovieSceneScriptingFloatChannel, default=1}
       leg_l_softness {type=MovieSceneScriptingFloatChannel, default=0}
       leg_l_pv_ik_ctrl.Location.X {type=MovieSceneScriptingFloatChannel, default=0}
       leg_l_pv_ik_ctrl.Location.Y {type=MovieSceneScriptingFloatChannel, default=0}
       leg_l_pv_ik_ctrl.Location.Z {type=MovieSceneScriptingFloatChannel, default=0}
       leg_l_pv_ik_ctrl.Rotation.X {type=MovieSceneScriptingFloatChannel, default=0}
       leg_l_pv_ik_ctrl.Rotation.Y {type=MovieSceneScriptingFloatChannel, default=0}
       leg_l_pv_ik_ctrl.Rotation.Z {type=MovieSceneScriptingFloatChannel, default=0}
       leg_l_pv_ik_ctrl.Scale.X {type=MovieSceneScriptingFloatChannel, default=1}
       leg_l_pv_ik_ctrl.Scale.Y {type=MovieSceneScriptingFloatChannel, default=1}
       leg_l_pv_ik_ctrl.Scale.Z {type=MovieSceneScriptingFloatChannel, default=1}
       thigh_r_fk_ctrl.Location.X {type=MovieSceneScriptingFloatChannel, default=0}
       thigh_r_fk_ctrl.Location.Y {type=MovieSceneScriptingFloatChannel, default=0}
       thigh_r_fk_ctrl.Location.Z {type=MovieSceneScriptingFloatChannel, default=0}
       thigh_r_fk_ctrl.Rotation.X {type=MovieSceneScriptingFloatChannel, default=1.2139}
       thigh_r_fk_ctrl.Rotation.Y {type=MovieSceneScriptingFloatChannel, default=-1.213}
       thigh_r_fk_ctrl.Rotation.Z {type=MovieSceneScriptingFloatChannel, default=-0.1592}
       thigh_r_fk_ctrl.Scale.X {type=MovieSceneScriptingFloatChannel, default=1}
       thigh_r_fk_ctrl.Scale.Y {type=MovieSceneScriptingFloatChannel, default=1}
       thigh_r_fk_ctrl.Scale.Z {type=MovieSceneScriptingFloatChannel, default=1}
       calf_r_fk_ctrl.Location.X {type=MovieSceneScriptingFloatChannel, default=0}
       calf_r_fk_ctrl.Location.Y {type=MovieSceneScriptingFloatChannel, default=0}
       calf_r_fk_ctrl.Location.Z {type=MovieSceneScriptingFloatChannel, default=0}
       calf_r_fk_ctrl.Rotation.X {type=MovieSceneScriptingFloatChannel, default=0.0561}
       calf_r_fk_ctrl.Rotation.Y {type=MovieSceneScriptingFloatChannel, default=1.1982}
       calf_r_fk_ctrl.Rotation.Z {type=MovieSceneScriptingFloatChannel, default=0.2616}
       calf_r_fk_ctrl.Scale.X {type=MovieSceneScriptingFloatChannel, default=1}
       calf_r_fk_ctrl.Scale.Y {type=MovieSceneScriptingFloatChannel, default=1}
       calf_r_fk_ctrl.Scale.Z {type=MovieSceneScriptingFloatChannel, default=1}
       foot_r_fk_ctrl.Location.X {type=MovieSceneScriptingFloatChannel, default=0}
       foot_r_fk_ctrl.Location.Y {type=MovieSceneScriptingFloatChannel, default=-0.0329}
       foot_r_fk_ctrl.Location.Z {type=MovieSceneScriptingFloatChannel, default=-0.9075}
       foot_r_fk_ctrl.Rotation.X {type=MovieSceneScriptingFloatChannel, default=-1.1648}
       foot_r_fk_ctrl.Rotation.Y {type=MovieSceneScriptingFloatChannel, default=0.1067}
       foot_r_fk_ctrl.Rotation.Z {type=MovieSceneScriptingFloatChannel, default=-0.1036}
       foot_r_fk_ctrl.Scale.X {type=MovieSceneScriptingFloatChannel, default=1}
       foot_r_fk_ctrl.Scale.Y {type=MovieSceneScriptingFloatChannel, default=1}
       foot_r_fk_ctrl.Scale.Z {type=MovieSceneScriptingFloatChannel, default=1}
       ball_r_fk_ctrl.Location.X {type=MovieSceneScriptingFloatChannel, default=0}
       ball_r_fk_ctrl.Location.Y {type=MovieSceneScriptingFloatChannel, default=0}
       ball_r_fk_ctrl.Location.Z {type=MovieSceneScriptingFloatChannel, default=0}
       ball_r_fk_ctrl.Rotation.X {type=MovieSceneScriptingFloatChannel, default=0}
       ball_r_fk_ctrl.Rotation.Y {type=MovieSceneScriptingFloatChannel, default=0}
       ball_r_fk_ctrl.Rotation.Z {type=MovieSceneScriptingFloatChannel, default=0}
       ball_r_fk_ctrl.Scale.X {type=MovieSceneScriptingFloatChannel, default=1}
       ball_r_fk_ctrl.Scale.Y {type=MovieSceneScriptingFloatChannel, default=1}
       ball_r_fk_ctrl.Scale.Z {type=MovieSceneScriptingFloatChannel, default=1}
       foot_r_ik_ctrl.Location.X {type=MovieSceneScriptingFloatChannel, default=0}
       foot_r_ik_ctrl.Location.Y {type=MovieSceneScriptingFloatChannel, default=0}
       foot_r_ik_ctrl.Location.Z {type=MovieSceneScriptingFloatChannel, default=0}
       foot_r_ik_ctrl.Rotation.X {type=MovieSceneScriptingFloatChannel, default=0}
       foot_r_ik_ctrl.Rotation.Y {type=MovieSceneScriptingFloatChannel, default=0}
       foot_r_ik_ctrl.Rotation.Z {type=MovieSceneScriptingFloatChannel, default=0}
       foot_r_ik_ctrl.Scale.X {type=MovieSceneScriptingFloatChannel, default=1}
       foot_r_ik_ctrl.Scale.Y {type=MovieSceneScriptingFloatChannel, default=1}
       foot_r_ik_ctrl.Scale.Z {type=MovieSceneScriptingFloatChannel, default=1}
       ball_r_ik_ctrl.Location.X {type=MovieSceneScriptingFloatChannel, default=0}
       ball_r_ik_ctrl.Location.Y {type=MovieSceneScriptingFloatChannel, default=0}
       ball_r_ik_ctrl.Location.Z {type=MovieSceneScriptingFloatChannel, default=0}
       ball_r_ik_ctrl.Rotation.X {type=MovieSceneScriptingFloatChannel, default=0}
       ball_r_ik_ctrl.Rotation.Y {type=MovieSceneScriptingFloatChannel, default=0}
       ball_r_ik_ctrl.Rotation.Z {type=MovieSceneScriptingFloatChannel, default=0}
       ball_r_ik_ctrl.Scale.X {type=MovieSceneScriptingFloatChannel, default=1}
       ball_r_ik_ctrl.Scale.Y {type=MovieSceneScriptingFloatChannel, default=1}
       ball_r_ik_ctrl.Scale.Z {type=MovieSceneScriptingFloatChannel, default=1}
       foot_bk1_r_ctrl.Location.X {type=MovieSceneScriptingFloatChannel, default=0}
       foot_bk1_r_ctrl.Location.Y {type=MovieSceneScriptingFloatChannel, default=0}
       foot_bk1_r_ctrl.Location.Z {type=MovieSceneScriptingFloatChannel, default=0}
       foot_bk1_r_ctrl.Rotation.X {type=MovieSceneScriptingFloatChannel, default=0}
       foot_bk1_r_ctrl.Rotation.Y {type=MovieSceneScriptingFloatChannel, default=0}
       foot_bk1_r_ctrl.Rotation.Z {type=MovieSceneScriptingFloatChannel, default=0}
       foot_bk1_r_ctrl.Scale.X {type=MovieSceneScriptingFloatChannel, default=1}
       foot_bk1_r_ctrl.Scale.Y {type=MovieSceneScriptingFloatChannel, default=1}
       foot_bk1_r_ctrl.Scale.Z {type=MovieSceneScriptingFloatChannel, default=1}
       foot_roll_r_ctrl.Location.X {type=MovieSceneScriptingFloatChannel, default=0}
       foot_roll_r_ctrl.Location.Y {type=MovieSceneScriptingFloatChannel, default=0}
       foot_roll_r_ctrl.Location.Z {type=MovieSceneScriptingFloatChannel, default=0}
       foot_roll_r_ctrl.Rotation.X {type=MovieSceneScriptingFloatChannel, default=0}
       foot_roll_r_ctrl.Rotation.Y {type=MovieSceneScriptingFloatChannel, default=0}
       foot_roll_r_ctrl.Rotation.Z {type=MovieSceneScriptingFloatChannel, default=0}
       foot_roll_r_ctrl.Scale.X {type=MovieSceneScriptingFloatChannel, default=1}
       foot_roll_r_ctrl.Scale.Y {type=MovieSceneScriptingFloatChannel, default=1}
       foot_roll_r_ctrl.Scale.Z {type=MovieSceneScriptingFloatChannel, default=1}
       roll_blend_r_ctrl {type=MovieSceneScriptingFloatChannel, default=1}
       tip_r_ctrl.Location.X {type=MovieSceneScriptingFloatChannel, default=0}
       tip_r_ctrl.Location.Y {type=MovieSceneScriptingFloatChannel, default=0}
       tip_r_ctrl.Location.Z {type=MovieSceneScriptingFloatChannel, default=0}
       tip_r_ctrl.Rotation.X {type=MovieSceneScriptingFloatChannel, default=0}
       tip_r_ctrl.Rotation.Y {type=MovieSceneScriptingFloatChannel, default=0}
       tip_r_ctrl.Rotation.Z {type=MovieSceneScriptingFloatChannel, default=0}
       tip_r_ctrl.Scale.X {type=MovieSceneScriptingFloatChannel, default=1}
       tip_r_ctrl.Scale.Y {type=MovieSceneScriptingFloatChannel, default=1}
       tip_r_ctrl.Scale.Z {type=MovieSceneScriptingFloatChannel, default=1}
       heel_r_ctrl.Location.X {type=MovieSceneScriptingFloatChannel, default=0}
       heel_r_ctrl.Location.Y {type=MovieSceneScriptingFloatChannel, default=0}
       heel_r_ctrl.Location.Z {type=MovieSceneScriptingFloatChannel, default=0}
       heel_r_ctrl.Rotation.X {type=MovieSceneScriptingFloatChannel, default=0}
       heel_r_ctrl.Rotation.Y {type=MovieSceneScriptingFloatChannel, default=0}
       heel_r_ctrl.Rotation.Z {type=MovieSceneScriptingFloatChannel, default=0}
       heel_r_ctrl.Scale.X {type=MovieSceneScriptingFloatChannel, default=1}
       heel_r_ctrl.Scale.Y {type=MovieSceneScriptingFloatChannel, default=1}
       heel_r_ctrl.Scale.Z {type=MovieSceneScriptingFloatChannel, default=1}
       leg_r_softness {type=MovieSceneScriptingFloatChannel, default=0}
       leg_r_pv_ik_ctrl.Location.X {type=MovieSceneScriptingFloatChannel, default=0}
       leg_r_pv_ik_ctrl.Location.Y {type=MovieSceneScriptingFloatChannel, default=0}
       leg_r_pv_ik_ctrl.Location.Z {type=MovieSceneScriptingFloatChannel, default=0}
       leg_r_pv_ik_ctrl.Rotation.X {type=MovieSceneScriptingFloatChannel, default=0}
       leg_r_pv_ik_ctrl.Rotation.Y {type=MovieSceneScriptingFloatChannel, default=0}
       leg_r_pv_ik_ctrl.Rotation.Z {type=MovieSceneScriptingFloatChannel, default=0}
       leg_r_pv_ik_ctrl.Scale.X {type=MovieSceneScriptingFloatChannel, default=1}
       leg_r_pv_ik_ctrl.Scale.Y {type=MovieSceneScriptingFloatChannel, default=1}
       leg_r_pv_ik_ctrl.Scale.Z {type=MovieSceneScriptingFloatChannel, default=1}
       upperarm_l_fk_ctrl.Location.X {type=MovieSceneScriptingFloatChannel, default=0}
       upperarm_l_fk_ctrl.Location.Y {type=MovieSceneScriptingFloatChannel, default=0}
       upperarm_l_fk_ctrl.Location.Z {type=MovieSceneScriptingFloatChannel, default=0}
       upperarm_l_fk_ctrl.Rotation.X {type=MovieSceneScriptingFloatChannel, default=0}
       upperarm_l_fk_ctrl.Rotation.Y {type=MovieSceneScriptingFloatChannel, default=0}
       upperarm_l_fk_ctrl.Rotation.Z {type=MovieSceneScriptingFloatChannel, default=0}
       upperarm_l_fk_ctrl.Scale.X {type=MovieSceneScriptingFloatChannel, default=1}
       upperarm_l_fk_ctrl.Scale.Y {type=MovieSceneScriptingFloatChannel, default=1}
       upperarm_l_fk_ctrl.Scale.Z {type=MovieSceneScriptingFloatChannel, default=1}
       lowerarm_l_fk_ctrl.Location.X {type=MovieSceneScriptingFloatChannel, default=0}
       lowerarm_l_fk_ctrl.Location.Y {type=MovieSceneScriptingFloatChannel, default=0}
       lowerarm_l_fk_ctrl.Location.Z {type=MovieSceneScriptingFloatChannel, default=0}
       lowerarm_l_fk_ctrl.Rotation.X {type=MovieSceneScriptingFloatChannel, default=0}
       lowerarm_l_fk_ctrl.Rotation.Y {type=MovieSceneScriptingFloatChannel, default=0}
       lowerarm_l_fk_ctrl.Rotation.Z {type=MovieSceneScriptingFloatChannel, default=0}
       lowerarm_l_fk_ctrl.Scale.X {type=MovieSceneScriptingFloatChannel, default=1}
       lowerarm_l_fk_ctrl.Scale.Y {type=MovieSceneScriptingFloatChannel, default=1}
       lowerarm_l_fk_ctrl.Scale.Z {type=MovieSceneScriptingFloatChannel, default=1}
       hand_l_fk_ctrl.Location.X {type=MovieSceneScriptingFloatChannel, default=0}
       hand_l_fk_ctrl.Location.Y {type=MovieSceneScriptingFloatChannel, default=0}
       hand_l_fk_ctrl.Location.Z {type=MovieSceneScriptingFloatChannel, default=0}
       hand_l_fk_ctrl.Rotation.X {type=MovieSceneScriptingFloatChannel, default=0}
       hand_l_fk_ctrl.Rotation.Y {type=MovieSceneScriptingFloatChannel, default=0}
       hand_l_fk_ctrl.Rotation.Z {type=MovieSceneScriptingFloatChannel, default=0}
       hand_l_fk_ctrl.Scale.X {type=MovieSceneScriptingFloatChannel, default=1}
       hand_l_fk_ctrl.Scale.Y {type=MovieSceneScriptingFloatChannel, default=1}
       hand_l_fk_ctrl.Scale.Z {type=MovieSceneScriptingFloatChannel, default=1}
       thumb_01_l_ctrl.Location.X {type=MovieSceneScriptingFloatChannel, default=0}
       thumb_01_l_ctrl.Location.Y {type=MovieSceneScriptingFloatChannel, default=0}
       thumb_01_l_ctrl.Location.Z {type=MovieSceneScriptingFloatChannel, default=0}
       thumb_01_l_ctrl.Rotation.X {type=MovieSceneScriptingFloatChannel, default=0}
       thumb_01_l_ctrl.Rotation.Y {type=MovieSceneScriptingFloatChannel, default=0}
       thumb_01_l_ctrl.Rotation.Z {type=MovieSceneScriptingFloatChannel, default=0}
       thumb_01_l_ctrl.Scale.X {type=MovieSceneScriptingFloatChannel, default=1}
       thumb_01_l_ctrl.Scale.Y {type=MovieSceneScriptingFloatChannel, default=1}
       thumb_01_l_ctrl.Scale.Z {type=MovieSceneScriptingFloatChannel, default=1}
       thumb_02_l_ctrl.Location.X {type=MovieSceneScriptingFloatChannel, default=0}
       thumb_02_l_ctrl.Location.Y {type=MovieSceneScriptingFloatChannel, default=0}
       thumb_02_l_ctrl.Location.Z {type=MovieSceneScriptingFloatChannel, default=0}
       thumb_02_l_ctrl.Rotation.X {type=MovieSceneScriptingFloatChannel, default=0}
       thumb_02_l_ctrl.Rotation.Y {type=MovieSceneScriptingFloatChannel, default=0}
       thumb_02_l_ctrl.Rotation.Z {type=MovieSceneScriptingFloatChannel, default=0}
       thumb_02_l_ctrl.Scale.X {type=MovieSceneScriptingFloatChannel, default=1}
       thumb_02_l_ctrl.Scale.Y {type=MovieSceneScriptingFloatChannel, default=1}
       thumb_02_l_ctrl.Scale.Z {type=MovieSceneScriptingFloatChannel, default=1}
       thumb_03_l_ctrl.Location.X {type=MovieSceneScriptingFloatChannel, default=0}
       thumb_03_l_ctrl.Location.Y {type=MovieSceneScriptingFloatChannel, default=0}
       thumb_03_l_ctrl.Location.Z {type=MovieSceneScriptingFloatChannel, default=0}
       thumb_03_l_ctrl.Rotation.X {type=MovieSceneScriptingFloatChannel, default=0}
       thumb_03_l_ctrl.Rotation.Y {type=MovieSceneScriptingFloatChannel, default=0}
       thumb_03_l_ctrl.Rotation.Z {type=MovieSceneScriptingFloatChannel, default=0}
       thumb_03_l_ctrl.Scale.X {type=MovieSceneScriptingFloatChannel, default=1}
       thumb_03_l_ctrl.Scale.Y {type=MovieSceneScriptingFloatChannel, default=1}
       thumb_03_l_ctrl.Scale.Z {type=MovieSceneScriptingFloatChannel, default=1}
       index_metacarpal_l_ctrl.Location.X {type=MovieSceneScriptingFloatChannel, default=0}
       index_metacarpal_l_ctrl.Location.Y {type=MovieSceneScriptingFloatChannel, default=0}
       index_metacarpal_l_ctrl.Location.Z {type=MovieSceneScriptingFloatChannel, default=0}
       index_metacarpal_l_ctrl.Rotation.X {type=MovieSceneScriptingFloatChannel, default=0}
       index_metacarpal_l_ctrl.Rotation.Y {type=MovieSceneScriptingFloatChannel, default=0}
       index_metacarpal_l_ctrl.Rotation.Z {type=MovieSceneScriptingFloatChannel, default=0}
       index_metacarpal_l_ctrl.Scale.X {type=MovieSceneScriptingFloatChannel, default=1}
       index_metacarpal_l_ctrl.Scale.Y {type=MovieSceneScriptingFloatChannel, default=1}
       index_metacarpal_l_ctrl.Scale.Z {type=MovieSceneScriptingFloatChannel, default=1}
       index_01_l_ctrl.Location.X {type=MovieSceneScriptingFloatChannel, default=0}
       index_01_l_ctrl.Location.Y {type=MovieSceneScriptingFloatChannel, default=0}
       index_01_l_ctrl.Location.Z {type=MovieSceneScriptingFloatChannel, default=0}
       index_01_l_ctrl.Rotation.X {type=MovieSceneScriptingFloatChannel, default=0}
       index_01_l_ctrl.Rotation.Y {type=MovieSceneScriptingFloatChannel, default=0}
       index_01_l_ctrl.Rotation.Z {type=MovieSceneScriptingFloatChannel, default=0}
       index_01_l_ctrl.Scale.X {type=MovieSceneScriptingFloatChannel, default=1}
       index_01_l_ctrl.Scale.Y {type=MovieSceneScriptingFloatChannel, default=1}
       index_01_l_ctrl.Scale.Z {type=MovieSceneScriptingFloatChannel, default=1}
       index_02_l_ctrl.Location.X {type=MovieSceneScriptingFloatChannel, default=0}
       index_02_l_ctrl.Location.Y {type=MovieSceneScriptingFloatChannel, default=0}
       index_02_l_ctrl.Location.Z {type=MovieSceneScriptingFloatChannel, default=0}
       index_02_l_ctrl.Rotation.X {type=MovieSceneScriptingFloatChannel, default=0}
       index_02_l_ctrl.Rotation.Y {type=MovieSceneScriptingFloatChannel, default=0}
       index_02_l_ctrl.Rotation.Z {type=MovieSceneScriptingFloatChannel, default=0}
       index_02_l_ctrl.Scale.X {type=MovieSceneScriptingFloatChannel, default=1}
       index_02_l_ctrl.Scale.Y {type=MovieSceneScriptingFloatChannel, default=1}
       index_02_l_ctrl.Scale.Z {type=MovieSceneScriptingFloatChannel, default=1}
       index_03_l_ctrl.Location.X {type=MovieSceneScriptingFloatChannel, default=0}
       index_03_l_ctrl.Location.Y {type=MovieSceneScriptingFloatChannel, default=0}
       index_03_l_ctrl.Location.Z {type=MovieSceneScriptingFloatChannel, default=0}
       index_03_l_ctrl.Rotation.X {type=MovieSceneScriptingFloatChannel, default=0}
       index_03_l_ctrl.Rotation.Y {type=MovieSceneScriptingFloatChannel, default=0}
       index_03_l_ctrl.Rotation.Z {type=MovieSceneScriptingFloatChannel, default=0}
       index_03_l_ctrl.Scale.X {type=MovieSceneScriptingFloatChannel, default=1}
       index_03_l_ctrl.Scale.Y {type=MovieSceneScriptingFloatChannel, default=1}
       index_03_l_ctrl.Scale.Z {type=MovieSceneScriptingFloatChannel, default=1}
       middle_metacarpal_l_ctrl.Location.X {type=MovieSceneScriptingFloatChannel, default=0}
       middle_metacarpal_l_ctrl.Location.Y {type=MovieSceneScriptingFloatChannel, default=0}
       middle_metacarpal_l_ctrl.Location.Z {type=MovieSceneScriptingFloatChannel, default=0}
       middle_metacarpal_l_ctrl.Rotation.X {type=MovieSceneScriptingFloatChannel, default=0}
       middle_metacarpal_l_ctrl.Rotation.Y {type=MovieSceneScriptingFloatChannel, default=0}
       middle_metacarpal_l_ctrl.Rotation.Z {type=MovieSceneScriptingFloatChannel, default=0}
       middle_metacarpal_l_ctrl.Scale.X {type=MovieSceneScriptingFloatChannel, default=1}
       middle_metacarpal_l_ctrl.Scale.Y {type=MovieSceneScriptingFloatChannel, default=1}
       middle_metacarpal_l_ctrl.Scale.Z {type=MovieSceneScriptingFloatChannel, default=1}
       middle_01_l_ctrl.Location.X {type=MovieSceneScriptingFloatChannel, default=0}
       middle_01_l_ctrl.Location.Y {type=MovieSceneScriptingFloatChannel, default=0}
       middle_01_l_ctrl.Location.Z {type=MovieSceneScriptingFloatChannel, default=0}
       middle_01_l_ctrl.Rotation.X {type=MovieSceneScriptingFloatChannel, default=0}
       middle_01_l_ctrl.Rotation.Y {type=MovieSceneScriptingFloatChannel, default=0}
       middle_01_l_ctrl.Rotation.Z {type=MovieSceneScriptingFloatChannel, default=0}
       middle_01_l_ctrl.Scale.X {type=MovieSceneScriptingFloatChannel, default=1}
       middle_01_l_ctrl.Scale.Y {type=MovieSceneScriptingFloatChannel, default=1}
       middle_01_l_ctrl.Scale.Z {type=MovieSceneScriptingFloatChannel, default=1}
       middle_02_l_ctrl.Location.X {type=MovieSceneScriptingFloatChannel, default=0}
       middle_02_l_ctrl.Location.Y {type=MovieSceneScriptingFloatChannel, default=0}
       middle_02_l_ctrl.Location.Z {type=MovieSceneScriptingFloatChannel, default=0}
       middle_02_l_ctrl.Rotation.X {type=MovieSceneScriptingFloatChannel, default=0}
       middle_02_l_ctrl.Rotation.Y {type=MovieSceneScriptingFloatChannel, default=0}
       middle_02_l_ctrl.Rotation.Z {type=MovieSceneScriptingFloatChannel, default=0}
       middle_02_l_ctrl.Scale.X {type=MovieSceneScriptingFloatChannel, default=1}
       middle_02_l_ctrl.Scale.Y {type=MovieSceneScriptingFloatChannel, default=1}
       middle_02_l_ctrl.Scale.Z {type=MovieSceneScriptingFloatChannel, default=1}
       middle_03_l_ctrl.Location.X {type=MovieSceneScriptingFloatChannel, default=0}
       middle_03_l_ctrl.Location.Y {type=MovieSceneScriptingFloatChannel, default=0}
       middle_03_l_ctrl.Location.Z {type=MovieSceneScriptingFloatChannel, default=0}
       middle_03_l_ctrl.Rotation.X {type=MovieSceneScriptingFloatChannel, default=0}
       middle_03_l_ctrl.Rotation.Y {type=MovieSceneScriptingFloatChannel, default=0}
       middle_03_l_ctrl.Rotation.Z {type=MovieSceneScriptingFloatChannel, default=0}
       middle_03_l_ctrl.Scale.X {type=MovieSceneScriptingFloatChannel, default=1}
       middle_03_l_ctrl.Scale.Y {type=MovieSceneScriptingFloatChannel, default=1}
       middle_03_l_ctrl.Scale.Z {type=MovieSceneScriptingFloatChannel, default=1}
       ring_metacarpal_l_ctrl.Location.X {type=MovieSceneScriptingFloatChannel, default=0}
       ring_metacarpal_l_ctrl.Location.Y {type=MovieSceneScriptingFloatChannel, default=0}
       ring_metacarpal_l_ctrl.Location.Z {type=MovieSceneScriptingFloatChannel, default=0}
       ring_metacarpal_l_ctrl.Rotation.X {type=MovieSceneScriptingFloatChannel, default=0}
       ring_metacarpal_l_ctrl.Rotation.Y {type=MovieSceneScriptingFloatChannel, default=0}
       ring_metacarpal_l_ctrl.Rotation.Z {type=MovieSceneScriptingFloatChannel, default=0}
       ring_metacarpal_l_ctrl.Scale.X {type=MovieSceneScriptingFloatChannel, default=1}
       ring_metacarpal_l_ctrl.Scale.Y {type=MovieSceneScriptingFloatChannel, default=1}
       ring_metacarpal_l_ctrl.Scale.Z {type=MovieSceneScriptingFloatChannel, default=1}
       ring_01_l_ctrl.Location.X {type=MovieSceneScriptingFloatChannel, default=0}
       ring_01_l_ctrl.Location.Y {type=MovieSceneScriptingFloatChannel, default=0}
       ring_01_l_ctrl.Location.Z {type=MovieSceneScriptingFloatChannel, default=0}
       ring_01_l_ctrl.Rotation.X {type=MovieSceneScriptingFloatChannel, default=0}
       ring_01_l_ctrl.Rotation.Y {type=MovieSceneScriptingFloatChannel, default=0}
       ring_01_l_ctrl.Rotation.Z {type=MovieSceneScriptingFloatChannel, default=0}
       ring_01_l_ctrl.Scale.X {type=MovieSceneScriptingFloatChannel, default=1}
       ring_01_l_ctrl.Scale.Y {type=MovieSceneScriptingFloatChannel, default=1}
       ring_01_l_ctrl.Scale.Z {type=MovieSceneScriptingFloatChannel, default=1}
       ring_02_l_ctrl.Location.X {type=MovieSceneScriptingFloatChannel, default=0}
       ring_02_l_ctrl.Location.Y {type=MovieSceneScriptingFloatChannel, default=0}
       ring_02_l_ctrl.Location.Z {type=MovieSceneScriptingFloatChannel, default=0}
       ring_02_l_ctrl.Rotation.X {type=MovieSceneScriptingFloatChannel, default=0}
       ring_02_l_ctrl.Rotation.Y {type=MovieSceneScriptingFloatChannel, default=0}
       ring_02_l_ctrl.Rotation.Z {type=MovieSceneScriptingFloatChannel, default=0}
       ring_02_l_ctrl.Scale.X {type=MovieSceneScriptingFloatChannel, default=1}
       ring_02_l_ctrl.Scale.Y {type=MovieSceneScriptingFloatChannel, default=1}
       ring_02_l_ctrl.Scale.Z {type=MovieSceneScriptingFloatChannel, default=1}
       ring_03_l_ctrl.Location.X {type=MovieSceneScriptingFloatChannel, default=0}
       ring_03_l_ctrl.Location.Y {type=MovieSceneScriptingFloatChannel, default=0}
       ring_03_l_ctrl.Location.Z {type=MovieSceneScriptingFloatChannel, default=0}
       ring_03_l_ctrl.Rotation.X {type=MovieSceneScriptingFloatChannel, default=0}
       ring_03_l_ctrl.Rotation.Y {type=MovieSceneScriptingFloatChannel, default=0}
       ring_03_l_ctrl.Rotation.Z {type=MovieSceneScriptingFloatChannel, default=0}
       ring_03_l_ctrl.Scale.X {type=MovieSceneScriptingFloatChannel, default=1}
       ring_03_l_ctrl.Scale.Y {type=MovieSceneScriptingFloatChannel, default=1}
       ring_03_l_ctrl.Scale.Z {type=MovieSceneScriptingFloatChannel, default=1}
       pinky_metacarpal_l_ctrl.Location.X {type=MovieSceneScriptingFloatChannel, default=0}
       pinky_metacarpal_l_ctrl.Location.Y {type=MovieSceneScriptingFloatChannel, default=0}
       pinky_metacarpal_l_ctrl.Location.Z {type=MovieSceneScriptingFloatChannel, default=0}
       pinky_metacarpal_l_ctrl.Rotation.X {type=MovieSceneScriptingFloatChannel, default=0}
       pinky_metacarpal_l_ctrl.Rotation.Y {type=MovieSceneScriptingFloatChannel, default=0}
       pinky_metacarpal_l_ctrl.Rotation.Z {type=MovieSceneScriptingFloatChannel, default=0}
       pinky_metacarpal_l_ctrl.Scale.X {type=MovieSceneScriptingFloatChannel, default=1}
       pinky_metacarpal_l_ctrl.Scale.Y {type=MovieSceneScriptingFloatChannel, default=1}
       pinky_metacarpal_l_ctrl.Scale.Z {type=MovieSceneScriptingFloatChannel, default=1}
       pinky_01_l_ctrl.Location.X {type=MovieSceneScriptingFloatChannel, default=0}
       pinky_01_l_ctrl.Location.Y {type=MovieSceneScriptingFloatChannel, default=0}
       pinky_01_l_ctrl.Location.Z {type=MovieSceneScriptingFloatChannel, default=0}
       pinky_01_l_ctrl.Rotation.X {type=MovieSceneScriptingFloatChannel, default=0}
       pinky_01_l_ctrl.Rotation.Y {type=MovieSceneScriptingFloatChannel, default=0}
       pinky_01_l_ctrl.Rotation.Z {type=MovieSceneScriptingFloatChannel, default=0}
       pinky_01_l_ctrl.Scale.X {type=MovieSceneScriptingFloatChannel, default=1}
       pinky_01_l_ctrl.Scale.Y {type=MovieSceneScriptingFloatChannel, default=1}
       pinky_01_l_ctrl.Scale.Z {type=MovieSceneScriptingFloatChannel, default=1}
       pinky_02_l_ctrl.Location.X {type=MovieSceneScriptingFloatChannel, default=0}
       pinky_02_l_ctrl.Location.Y {type=MovieSceneScriptingFloatChannel, default=0}
       pinky_02_l_ctrl.Location.Z {type=MovieSceneScriptingFloatChannel, default=0}
       pinky_02_l_ctrl.Rotation.X {type=MovieSceneScriptingFloatChannel, default=0}
       pinky_02_l_ctrl.Rotation.Y {type=MovieSceneScriptingFloatChannel, default=0}
       pinky_02_l_ctrl.Rotation.Z {type=MovieSceneScriptingFloatChannel, default=0}
       pinky_02_l_ctrl.Scale.X {type=MovieSceneScriptingFloatChannel, default=1}
       pinky_02_l_ctrl.Scale.Y {type=MovieSceneScriptingFloatChannel, default=1}
       pinky_02_l_ctrl.Scale.Z {type=MovieSceneScriptingFloatChannel, default=1}
       pinky_03_l_ctrl.Location.X {type=MovieSceneScriptingFloatChannel, default=0}
       pinky_03_l_ctrl.Location.Y {type=MovieSceneScriptingFloatChannel, default=0}
       pinky_03_l_ctrl.Location.Z {type=MovieSceneScriptingFloatChannel, default=0}
       pinky_03_l_ctrl.Rotation.X {type=MovieSceneScriptingFloatChannel, default=0}
       pinky_03_l_ctrl.Rotation.Y {type=MovieSceneScriptingFloatChannel, default=0}
       pinky_03_l_ctrl.Rotation.Z {type=MovieSceneScriptingFloatChannel, default=0}
       pinky_03_l_ctrl.Scale.X {type=MovieSceneScriptingFloatChannel, default=1}
       pinky_03_l_ctrl.Scale.Y {type=MovieSceneScriptingFloatChannel, default=1}
       pinky_03_l_ctrl.Scale.Z {type=MovieSceneScriptingFloatChannel, default=1}
       hand_l_ik_ctrl.Location.X {type=MovieSceneScriptingFloatChannel, default=0}
       hand_l_ik_ctrl.Location.Y {type=MovieSceneScriptingFloatChannel, default=0}
       hand_l_ik_ctrl.Location.Z {type=MovieSceneScriptingFloatChannel, default=0}
       hand_l_ik_ctrl.Rotation.X {type=MovieSceneScriptingFloatChannel, default=0}
       hand_l_ik_ctrl.Rotation.Y {type=MovieSceneScriptingFloatChannel, default=0}
       hand_l_ik_ctrl.Rotation.Z {type=MovieSceneScriptingFloatChannel, default=0}
       hand_l_ik_ctrl.Scale.X {type=MovieSceneScriptingFloatChannel, default=1}
       hand_l_ik_ctrl.Scale.Y {type=MovieSceneScriptingFloatChannel, default=1}
       hand_l_ik_ctrl.Scale.Z {type=MovieSceneScriptingFloatChannel, default=1}
       Softness {type=MovieSceneScriptingFloatChannel, default=0}
       hand_l_ikorient_ctrl.Rotation.Roll {type=MovieSceneScriptingFloatChannel, default=0}
       hand_l_ikorient_ctrl.Rotation.Pitch {type=MovieSceneScriptingFloatChannel, default=0}
       hand_l_ikorient_ctrl.Rotation.Yaw {type=MovieSceneScriptingFloatChannel, default=0}
       arm_l_pv_ik_ctrl.Location.X {type=MovieSceneScriptingFloatChannel, default=0}
       arm_l_pv_ik_ctrl.Location.Y {type=MovieSceneScriptingFloatChannel, default=0}
       arm_l_pv_ik_ctrl.Location.Z {type=MovieSceneScriptingFloatChannel, default=0}
       arm_l_pv_ik_ctrl.Rotation.X {type=MovieSceneScriptingFloatChannel, default=0}
       arm_l_pv_ik_ctrl.Rotation.Y {type=MovieSceneScriptingFloatChannel, default=0}
       arm_l_pv_ik_ctrl.Rotation.Z {type=MovieSceneScriptingFloatChannel, default=0}
       arm_l_pv_ik_ctrl.Scale.X {type=MovieSceneScriptingFloatChannel, default=1}
       arm_l_pv_ik_ctrl.Scale.Y {type=MovieSceneScriptingFloatChannel, default=1}
       arm_l_pv_ik_ctrl.Scale.Z {type=MovieSceneScriptingFloatChannel, default=1}
       upperarm_r_fk_ctrl.Location.X {type=MovieSceneScriptingFloatChannel, default=0}
       upperarm_r_fk_ctrl.Location.Y {type=MovieSceneScriptingFloatChannel, default=0}
       upperarm_r_fk_ctrl.Location.Z {type=MovieSceneScriptingFloatChannel, default=0}
       upperarm_r_fk_ctrl.Rotation.X {type=MovieSceneScriptingFloatChannel, default=0}
       upperarm_r_fk_ctrl.Rotation.Y {type=MovieSceneScriptingFloatChannel, default=0}
       upperarm_r_fk_ctrl.Rotation.Z {type=MovieSceneScriptingFloatChannel, default=0}
       upperarm_r_fk_ctrl.Scale.X {type=MovieSceneScriptingFloatChannel, default=1}
       upperarm_r_fk_ctrl.Scale.Y {type=MovieSceneScriptingFloatChannel, default=1}
       upperarm_r_fk_ctrl.Scale.Z {type=MovieSceneScriptingFloatChannel, default=1}
       lowerarm_r_fk_ctrl.Location.X {type=MovieSceneScriptingFloatChannel, default=0}
       lowerarm_r_fk_ctrl.Location.Y {type=MovieSceneScriptingFloatChannel, default=0}
       lowerarm_r_fk_ctrl.Location.Z {type=MovieSceneScriptingFloatChannel, default=0}
       lowerarm_r_fk_ctrl.Rotation.X {type=MovieSceneScriptingFloatChannel, default=0}
       lowerarm_r_fk_ctrl.Rotation.Y {type=MovieSceneScriptingFloatChannel, default=0}
       lowerarm_r_fk_ctrl.Rotation.Z {type=MovieSceneScriptingFloatChannel, default=0}
       lowerarm_r_fk_ctrl.Scale.X {type=MovieSceneScriptingFloatChannel, default=1}
       lowerarm_r_fk_ctrl.Scale.Y {type=MovieSceneScriptingFloatChannel, default=1}
       lowerarm_r_fk_ctrl.Scale.Z {type=MovieSceneScriptingFloatChannel, default=1}
       hand_r_fk_ctrl.Location.X {type=MovieSceneScriptingFloatChannel, default=0}
       hand_r_fk_ctrl.Location.Y {type=MovieSceneScriptingFloatChannel, default=0}
       hand_r_fk_ctrl.Location.Z {type=MovieSceneScriptingFloatChannel, default=0}
       hand_r_fk_ctrl.Rotation.X {type=MovieSceneScriptingFloatChannel, default=0}
       hand_r_fk_ctrl.Rotation.Y {type=MovieSceneScriptingFloatChannel, default=0}
       hand_r_fk_ctrl.Rotation.Z {type=MovieSceneScriptingFloatChannel, default=0}
       hand_r_fk_ctrl.Scale.X {type=MovieSceneScriptingFloatChannel, default=1}
       hand_r_fk_ctrl.Scale.Y {type=MovieSceneScriptingFloatChannel, default=1}
       hand_r_fk_ctrl.Scale.Z {type=MovieSceneScriptingFloatChannel, default=1}
       thumb_01_r_ctrl.Location.X {type=MovieSceneScriptingFloatChannel, default=0}
       thumb_01_r_ctrl.Location.Y {type=MovieSceneScriptingFloatChannel, default=0}
       thumb_01_r_ctrl.Location.Z {type=MovieSceneScriptingFloatChannel, default=0}
       thumb_01_r_ctrl.Rotation.X {type=MovieSceneScriptingFloatChannel, default=0}
       thumb_01_r_ctrl.Rotation.Y {type=MovieSceneScriptingFloatChannel, default=0}
       thumb_01_r_ctrl.Rotation.Z {type=MovieSceneScriptingFloatChannel, default=0}
       thumb_01_r_ctrl.Scale.X {type=MovieSceneScriptingFloatChannel, default=1}
       thumb_01_r_ctrl.Scale.Y {type=MovieSceneScriptingFloatChannel, default=1}
       thumb_01_r_ctrl.Scale.Z {type=MovieSceneScriptingFloatChannel, default=1}
       thumb_02_r_ctrl.Location.X {type=MovieSceneScriptingFloatChannel, default=0}
       thumb_02_r_ctrl.Location.Y {type=MovieSceneScriptingFloatChannel, default=0}
       thumb_02_r_ctrl.Location.Z {type=MovieSceneScriptingFloatChannel, default=0}
       thumb_02_r_ctrl.Rotation.X {type=MovieSceneScriptingFloatChannel, default=0}
       thumb_02_r_ctrl.Rotation.Y {type=MovieSceneScriptingFloatChannel, default=0}
       thumb_02_r_ctrl.Rotation.Z {type=MovieSceneScriptingFloatChannel, default=0}
       thumb_02_r_ctrl.Scale.X {type=MovieSceneScriptingFloatChannel, default=1}
       thumb_02_r_ctrl.Scale.Y {type=MovieSceneScriptingFloatChannel, default=1}
       thumb_02_r_ctrl.Scale.Z {type=MovieSceneScriptingFloatChannel, default=1}
       thumb_03_r_ctrl.Location.X {type=MovieSceneScriptingFloatChannel, default=0}
       thumb_03_r_ctrl.Location.Y {type=MovieSceneScriptingFloatChannel, default=0}
       thumb_03_r_ctrl.Location.Z {type=MovieSceneScriptingFloatChannel, default=0}
       thumb_03_r_ctrl.Rotation.X {type=MovieSceneScriptingFloatChannel, default=0}
       thumb_03_r_ctrl.Rotation.Y {type=MovieSceneScriptingFloatChannel, default=0}
       thumb_03_r_ctrl.Rotation.Z {type=MovieSceneScriptingFloatChannel, default=0}
       thumb_03_r_ctrl.Scale.X {type=MovieSceneScriptingFloatChannel, default=1}
       thumb_03_r_ctrl.Scale.Y {type=MovieSceneScriptingFloatChannel, default=1}
       thumb_03_r_ctrl.Scale.Z {type=MovieSceneScriptingFloatChannel, default=1}
       index_metacarpal_r_ctrl.Location.X {type=MovieSceneScriptingFloatChannel, default=0}
       index_metacarpal_r_ctrl.Location.Y {type=MovieSceneScriptingFloatChannel, default=0}
       index_metacarpal_r_ctrl.Location.Z {type=MovieSceneScriptingFloatChannel, default=0}
       index_metacarpal_r_ctrl.Rotation.X {type=MovieSceneScriptingFloatChannel, default=0}
       index_metacarpal_r_ctrl.Rotation.Y {type=MovieSceneScriptingFloatChannel, default=0}
       index_metacarpal_r_ctrl.Rotation.Z {type=MovieSceneScriptingFloatChannel, default=0}
       index_metacarpal_r_ctrl.Scale.X {type=MovieSceneScriptingFloatChannel, default=1}
       index_metacarpal_r_ctrl.Scale.Y {type=MovieSceneScriptingFloatChannel, default=1}
       index_metacarpal_r_ctrl.Scale.Z {type=MovieSceneScriptingFloatChannel, default=1}
       index_01_r_ctrl.Location.X {type=MovieSceneScriptingFloatChannel, default=0}
       index_01_r_ctrl.Location.Y {type=MovieSceneScriptingFloatChannel, default=0}
       index_01_r_ctrl.Location.Z {type=MovieSceneScriptingFloatChannel, default=0}
       index_01_r_ctrl.Rotation.X {type=MovieSceneScriptingFloatChannel, default=0}
       index_01_r_ctrl.Rotation.Y {type=MovieSceneScriptingFloatChannel, default=0}
       index_01_r_ctrl.Rotation.Z {type=MovieSceneScriptingFloatChannel, default=0}
       index_01_r_ctrl.Scale.X {type=MovieSceneScriptingFloatChannel, default=1}
       index_01_r_ctrl.Scale.Y {type=MovieSceneScriptingFloatChannel, default=1}
       index_01_r_ctrl.Scale.Z {type=MovieSceneScriptingFloatChannel, default=1}
       index_02_r_ctrl.Location.X {type=MovieSceneScriptingFloatChannel, default=0}
       index_02_r_ctrl.Location.Y {type=MovieSceneScriptingFloatChannel, default=0}
       index_02_r_ctrl.Location.Z {type=MovieSceneScriptingFloatChannel, default=0}
       index_02_r_ctrl.Rotation.X {type=MovieSceneScriptingFloatChannel, default=0}
       index_02_r_ctrl.Rotation.Y {type=MovieSceneScriptingFloatChannel, default=0}
       index_02_r_ctrl.Rotation.Z {type=MovieSceneScriptingFloatChannel, default=0}
       index_02_r_ctrl.Scale.X {type=MovieSceneScriptingFloatChannel, default=1}
       index_02_r_ctrl.Scale.Y {type=MovieSceneScriptingFloatChannel, default=1}
       index_02_r_ctrl.Scale.Z {type=MovieSceneScriptingFloatChannel, default=1}
       index_03_r_ctrl.Location.X {type=MovieSceneScriptingFloatChannel, default=0}
       index_03_r_ctrl.Location.Y {type=MovieSceneScriptingFloatChannel, default=0}
       index_03_r_ctrl.Location.Z {type=MovieSceneScriptingFloatChannel, default=0}
       index_03_r_ctrl.Rotation.X {type=MovieSceneScriptingFloatChannel, default=0}
       index_03_r_ctrl.Rotation.Y {type=MovieSceneScriptingFloatChannel, default=0}
       index_03_r_ctrl.Rotation.Z {type=MovieSceneScriptingFloatChannel, default=0}
       index_03_r_ctrl.Scale.X {type=MovieSceneScriptingFloatChannel, default=1}
       index_03_r_ctrl.Scale.Y {type=MovieSceneScriptingFloatChannel, default=1}
       index_03_r_ctrl.Scale.Z {type=MovieSceneScriptingFloatChannel, default=1}
       middle_metacarpal_r_ctrl.Location.X {type=MovieSceneScriptingFloatChannel, default=0}
       middle_metacarpal_r_ctrl.Location.Y {type=MovieSceneScriptingFloatChannel, default=0}
       middle_metacarpal_r_ctrl.Location.Z {type=MovieSceneScriptingFloatChannel, default=0}
       middle_metacarpal_r_ctrl.Rotation.X {type=MovieSceneScriptingFloatChannel, default=0}
       middle_metacarpal_r_ctrl.Rotation.Y {type=MovieSceneScriptingFloatChannel, default=0}
       middle_metacarpal_r_ctrl.Rotation.Z {type=MovieSceneScriptingFloatChannel, default=0}
       middle_metacarpal_r_ctrl.Scale.X {type=MovieSceneScriptingFloatChannel, default=1}
       middle_metacarpal_r_ctrl.Scale.Y {type=MovieSceneScriptingFloatChannel, default=1}
       middle_metacarpal_r_ctrl.Scale.Z {type=MovieSceneScriptingFloatChannel, default=1}
       middle_01_r_ctrl.Location.X {type=MovieSceneScriptingFloatChannel, default=0}
       middle_01_r_ctrl.Location.Y {type=MovieSceneScriptingFloatChannel, default=0}
       middle_01_r_ctrl.Location.Z {type=MovieSceneScriptingFloatChannel, default=0}
       middle_01_r_ctrl.Rotation.X {type=MovieSceneScriptingFloatChannel, default=0}
       middle_01_r_ctrl.Rotation.Y {type=MovieSceneScriptingFloatChannel, default=0}
       middle_01_r_ctrl.Rotation.Z {type=MovieSceneScriptingFloatChannel, default=0}
       middle_01_r_ctrl.Scale.X {type=MovieSceneScriptingFloatChannel, default=1}
       middle_01_r_ctrl.Scale.Y {type=MovieSceneScriptingFloatChannel, default=1}
       middle_01_r_ctrl.Scale.Z {type=MovieSceneScriptingFloatChannel, default=1}
       middle_02_r_ctrl.Location.X {type=MovieSceneScriptingFloatChannel, default=0}
       middle_02_r_ctrl.Location.Y {type=MovieSceneScriptingFloatChannel, default=0}
       middle_02_r_ctrl.Location.Z {type=MovieSceneScriptingFloatChannel, default=0}
       middle_02_r_ctrl.Rotation.X {type=MovieSceneScriptingFloatChannel, default=0}
       middle_02_r_ctrl.Rotation.Y {type=MovieSceneScriptingFloatChannel, default=0}
       middle_02_r_ctrl.Rotation.Z {type=MovieSceneScriptingFloatChannel, default=0}
       middle_02_r_ctrl.Scale.X {type=MovieSceneScriptingFloatChannel, default=1}
       middle_02_r_ctrl.Scale.Y {type=MovieSceneScriptingFloatChannel, default=1}
       middle_02_r_ctrl.Scale.Z {type=MovieSceneScriptingFloatChannel, default=1}
       middle_03_r_ctrl.Location.X {type=MovieSceneScriptingFloatChannel, default=0}
       middle_03_r_ctrl.Location.Y {type=MovieSceneScriptingFloatChannel, default=0}
       middle_03_r_ctrl.Location.Z {type=MovieSceneScriptingFloatChannel, default=0}
       middle_03_r_ctrl.Rotation.X {type=MovieSceneScriptingFloatChannel, default=0}
       middle_03_r_ctrl.Rotation.Y {type=MovieSceneScriptingFloatChannel, default=0}
       middle_03_r_ctrl.Rotation.Z {type=MovieSceneScriptingFloatChannel, default=0}
       middle_03_r_ctrl.Scale.X {type=MovieSceneScriptingFloatChannel, default=1}
       middle_03_r_ctrl.Scale.Y {type=MovieSceneScriptingFloatChannel, default=1}
       middle_03_r_ctrl.Scale.Z {type=MovieSceneScriptingFloatChannel, default=1}
       ring_metacarpal_r_ctrl.Location.X {type=MovieSceneScriptingFloatChannel, default=0}
       ring_metacarpal_r_ctrl.Location.Y {type=MovieSceneScriptingFloatChannel, default=0}
       ring_metacarpal_r_ctrl.Location.Z {type=MovieSceneScriptingFloatChannel, default=0}
       ring_metacarpal_r_ctrl.Rotation.X {type=MovieSceneScriptingFloatChannel, default=0}
       ring_metacarpal_r_ctrl.Rotation.Y {type=MovieSceneScriptingFloatChannel, default=0}
       ring_metacarpal_r_ctrl.Rotation.Z {type=MovieSceneScriptingFloatChannel, default=0}
       ring_metacarpal_r_ctrl.Scale.X {type=MovieSceneScriptingFloatChannel, default=1}
       ring_metacarpal_r_ctrl.Scale.Y {type=MovieSceneScriptingFloatChannel, default=1}
       ring_metacarpal_r_ctrl.Scale.Z {type=MovieSceneScriptingFloatChannel, default=1}
       ring_01_r_ctrl.Location.X {type=MovieSceneScriptingFloatChannel, default=0}
       ring_01_r_ctrl.Location.Y {type=MovieSceneScriptingFloatChannel, default=0}
       ring_01_r_ctrl.Location.Z {type=MovieSceneScriptingFloatChannel, default=0}
       ring_01_r_ctrl.Rotation.X {type=MovieSceneScriptingFloatChannel, default=0}
       ring_01_r_ctrl.Rotation.Y {type=MovieSceneScriptingFloatChannel, default=0}
       ring_01_r_ctrl.Rotation.Z {type=MovieSceneScriptingFloatChannel, default=0}
       ring_01_r_ctrl.Scale.X {type=MovieSceneScriptingFloatChannel, default=1}
       ring_01_r_ctrl.Scale.Y {type=MovieSceneScriptingFloatChannel, default=1}
       ring_01_r_ctrl.Scale.Z {type=MovieSceneScriptingFloatChannel, default=1}
       ring_02_r_ctrl.Location.X {type=MovieSceneScriptingFloatChannel, default=0}
       ring_02_r_ctrl.Location.Y {type=MovieSceneScriptingFloatChannel, default=0}
       ring_02_r_ctrl.Location.Z {type=MovieSceneScriptingFloatChannel, default=0}
       ring_02_r_ctrl.Rotation.X {type=MovieSceneScriptingFloatChannel, default=0}
       ring_02_r_ctrl.Rotation.Y {type=MovieSceneScriptingFloatChannel, default=0}
       ring_02_r_ctrl.Rotation.Z {type=MovieSceneScriptingFloatChannel, default=0}
       ring_02_r_ctrl.Scale.X {type=MovieSceneScriptingFloatChannel, default=1}
       ring_02_r_ctrl.Scale.Y {type=MovieSceneScriptingFloatChannel, default=1}
       ring_02_r_ctrl.Scale.Z {type=MovieSceneScriptingFloatChannel, default=1}
       ring_03_r_ctrl.Location.X {type=MovieSceneScriptingFloatChannel, default=0}
       ring_03_r_ctrl.Location.Y {type=MovieSceneScriptingFloatChannel, default=0}
       ring_03_r_ctrl.Location.Z {type=MovieSceneScriptingFloatChannel, default=0}
       ring_03_r_ctrl.Rotation.X {type=MovieSceneScriptingFloatChannel, default=0}
       ring_03_r_ctrl.Rotation.Y {type=MovieSceneScriptingFloatChannel, default=0}
       ring_03_r_ctrl.Rotation.Z {type=MovieSceneScriptingFloatChannel, default=0}
       ring_03_r_ctrl.Scale.X {type=MovieSceneScriptingFloatChannel, default=1}
       ring_03_r_ctrl.Scale.Y {type=MovieSceneScriptingFloatChannel, default=1}
       ring_03_r_ctrl.Scale.Z {type=MovieSceneScriptingFloatChannel, default=1}
       pinky_metacarpal_r_ctrl.Location.X {type=MovieSceneScriptingFloatChannel, default=0}
       pinky_metacarpal_r_ctrl.Location.Y {type=MovieSceneScriptingFloatChannel, default=0}
       pinky_metacarpal_r_ctrl.Location.Z {type=MovieSceneScriptingFloatChannel, default=0}
       pinky_metacarpal_r_ctrl.Rotation.X {type=MovieSceneScriptingFloatChannel, default=0}
       pinky_metacarpal_r_ctrl.Rotation.Y {type=MovieSceneScriptingFloatChannel, default=0}
       pinky_metacarpal_r_ctrl.Rotation.Z {type=MovieSceneScriptingFloatChannel, default=0}
       pinky_metacarpal_r_ctrl.Scale.X {type=MovieSceneScriptingFloatChannel, default=1}
       pinky_metacarpal_r_ctrl.Scale.Y {type=MovieSceneScriptingFloatChannel, default=1}
       pinky_metacarpal_r_ctrl.Scale.Z {type=MovieSceneScriptingFloatChannel, default=1}
       pinky_01_r_ctrl.Location.X {type=MovieSceneScriptingFloatChannel, default=0}
       pinky_01_r_ctrl.Location.Y {type=MovieSceneScriptingFloatChannel, default=0}
       pinky_01_r_ctrl.Location.Z {type=MovieSceneScriptingFloatChannel, default=0}
       pinky_01_r_ctrl.Rotation.X {type=MovieSceneScriptingFloatChannel, default=0}
       pinky_01_r_ctrl.Rotation.Y {type=MovieSceneScriptingFloatChannel, default=0}
       pinky_01_r_ctrl.Rotation.Z {type=MovieSceneScriptingFloatChannel, default=0}
       pinky_01_r_ctrl.Scale.X {type=MovieSceneScriptingFloatChannel, default=1}
       pinky_01_r_ctrl.Scale.Y {type=MovieSceneScriptingFloatChannel, default=1}
       pinky_01_r_ctrl.Scale.Z {type=MovieSceneScriptingFloatChannel, default=1}
       pinky_02_r_ctrl.Location.X {type=MovieSceneScriptingFloatChannel, default=0}
       pinky_02_r_ctrl.Location.Y {type=MovieSceneScriptingFloatChannel, default=0}
       pinky_02_r_ctrl.Location.Z {type=MovieSceneScriptingFloatChannel, default=0}
       pinky_02_r_ctrl.Rotation.X {type=MovieSceneScriptingFloatChannel, default=0}
       pinky_02_r_ctrl.Rotation.Y {type=MovieSceneScriptingFloatChannel, default=0}
       pinky_02_r_ctrl.Rotation.Z {type=MovieSceneScriptingFloatChannel, default=0}
       pinky_02_r_ctrl.Scale.X {type=MovieSceneScriptingFloatChannel, default=1}
       pinky_02_r_ctrl.Scale.Y {type=MovieSceneScriptingFloatChannel, default=1}
       pinky_02_r_ctrl.Scale.Z {type=MovieSceneScriptingFloatChannel, default=1}
       pinky_03_r_ctrl.Location.X {type=MovieSceneScriptingFloatChannel, default=0}
       pinky_03_r_ctrl.Location.Y {type=MovieSceneScriptingFloatChannel, default=0}
       pinky_03_r_ctrl.Location.Z {type=MovieSceneScriptingFloatChannel, default=0}
       pinky_03_r_ctrl.Rotation.X {type=MovieSceneScriptingFloatChannel, default=0}
       pinky_03_r_ctrl.Rotation.Y {type=MovieSceneScriptingFloatChannel, default=0}
       pinky_03_r_ctrl.Rotation.Z {type=MovieSceneScriptingFloatChannel, default=0}
       pinky_03_r_ctrl.Scale.X {type=MovieSceneScriptingFloatChannel, default=1}
       pinky_03_r_ctrl.Scale.Y {type=MovieSceneScriptingFloatChannel, default=1}
       pinky_03_r_ctrl.Scale.Z {type=MovieSceneScriptingFloatChannel, default=1}
       hand_r_ik_ctrl.Location.X {type=MovieSceneScriptingFloatChannel, default=0}
       hand_r_ik_ctrl.Location.Y {type=MovieSceneScriptingFloatChannel, default=0}
       hand_r_ik_ctrl.Location.Z {type=MovieSceneScriptingFloatChannel, default=0}
       hand_r_ik_ctrl.Rotation.X {type=MovieSceneScriptingFloatChannel, default=0}
       hand_r_ik_ctrl.Rotation.Y {type=MovieSceneScriptingFloatChannel, default=0}
       hand_r_ik_ctrl.Rotation.Z {type=MovieSceneScriptingFloatChannel, default=0}
       hand_r_ik_ctrl.Scale.X {type=MovieSceneScriptingFloatChannel, default=1}
       hand_r_ik_ctrl.Scale.Y {type=MovieSceneScriptingFloatChannel, default=1}
       hand_r_ik_ctrl.Scale.Z {type=MovieSceneScriptingFloatChannel, default=1}
       Softness_2 {type=MovieSceneScriptingFloatChannel, default=0}
       hand_r_ikorient_ctrl.Rotation.Roll {type=MovieSceneScriptingFloatChannel, default=0}
       hand_r_ikorient_ctrl.Rotation.Pitch {type=MovieSceneScriptingFloatChannel, default=0}
       hand_r_ikorient_ctrl.Rotation.Yaw {type=MovieSceneScriptingFloatChannel, default=0}
       arm_r_pv_ik_ctrl.Location.X {type=MovieSceneScriptingFloatChannel, default=0}
       arm_r_pv_ik_ctrl.Location.Y {type=MovieSceneScriptingFloatChannel, default=0}
       arm_r_pv_ik_ctrl.Location.Z {type=MovieSceneScriptingFloatChannel, default=0}
       arm_r_pv_ik_ctrl.Rotation.X {type=MovieSceneScriptingFloatChannel, default=0}
       arm_r_pv_ik_ctrl.Rotation.Y {type=MovieSceneScriptingFloatChannel, default=0}
       arm_r_pv_ik_ctrl.Rotation.Z {type=MovieSceneScriptingFloatChannel, default=0}
       arm_r_pv_ik_ctrl.Scale.X {type=MovieSceneScriptingFloatChannel, default=1}
       arm_r_pv_ik_ctrl.Scale.Y {type=MovieSceneScriptingFloatChannel, default=1}
       arm_r_pv_ik_ctrl.Scale.Z {type=MovieSceneScriptingFloatChannel, default=1}
       body_ctrl.Location.X {type=MovieSceneScriptingFloatChannel, default=0}
       body_ctrl.Location.Y {type=MovieSceneScriptingFloatChannel, default=0}
       body_ctrl.Location.Z {type=MovieSceneScriptingFloatChannel, default=0}
       body_ctrl.Rotation.X {type=MovieSceneScriptingFloatChannel, default=0}
       body_ctrl.Rotation.Y {type=MovieSceneScriptingFloatChannel, default=0}
       body_ctrl.Rotation.Z {type=MovieSceneScriptingFloatChannel, default=0}
       body_ctrl.Scale.X {type=MovieSceneScriptingFloatChannel, default=1}
       body_ctrl.Scale.Y {type=MovieSceneScriptingFloatChannel, default=1}
       body_ctrl.Scale.Z {type=MovieSceneScriptingFloatChannel, default=1}
       hips_ctrl.Location.X {type=MovieSceneScriptingFloatChannel, default=0}
       hips_ctrl.Location.Y {type=MovieSceneScriptingFloatChannel, default=0}
       hips_ctrl.Location.Z {type=MovieSceneScriptingFloatChannel, default=0}
       hips_ctrl.Rotation.X {type=MovieSceneScriptingFloatChannel, default=0}
       hips_ctrl.Rotation.Y {type=MovieSceneScriptingFloatChannel, default=0}
       hips_ctrl.Rotation.Z {type=MovieSceneScriptingFloatChannel, default=0}
       hips_ctrl.Scale.X {type=MovieSceneScriptingFloatChannel, default=1}
       hips_ctrl.Scale.Y {type=MovieSceneScriptingFloatChannel, default=1}
       hips_ctrl.Scale.Z {type=MovieSceneScriptingFloatChannel, default=1}
       hips_tan_ctrl.Location.X {type=MovieSceneScriptingFloatChannel, default=0}
       hips_tan_ctrl.Location.Y {type=MovieSceneScriptingFloatChannel, default=0.008}
       hips_tan_ctrl.Location.Z {type=MovieSceneScriptingFloatChannel, default=-0.0073}
       hips_tan_ctrl.Rotation.X {type=MovieSceneScriptingFloatChannel, default=0}
       hips_tan_ctrl.Rotation.Y {type=MovieSceneScriptingFloatChannel, default=0}
       hips_tan_ctrl.Rotation.Z {type=MovieSceneScriptingFloatChannel, default=0}
       hips_tan_ctrl.Scale.X {type=MovieSceneScriptingFloatChannel, default=1}
       hips_tan_ctrl.Scale.Y {type=MovieSceneScriptingFloatChannel, default=1}
       hips_tan_ctrl.Scale.Z {type=MovieSceneScriptingFloatChannel, default=1}
       spine_01_ctrl.Location.X {type=MovieSceneScriptingFloatChannel, default=0}
       spine_01_ctrl.Location.Y {type=MovieSceneScriptingFloatChannel, default=0}
       spine_01_ctrl.Location.Z {type=MovieSceneScriptingFloatChannel, default=0}
       spine_01_ctrl.Rotation.X {type=MovieSceneScriptingFloatChannel, default=0}
       spine_01_ctrl.Rotation.Y {type=MovieSceneScriptingFloatChannel, default=0}
       spine_01_ctrl.Rotation.Z {type=MovieSceneScriptingFloatChannel, default=0}
       spine_01_ctrl.Scale.X {type=MovieSceneScriptingFloatChannel, default=1}
       spine_01_ctrl.Scale.Y {type=MovieSceneScriptingFloatChannel, default=1}
       spine_01_ctrl.Scale.Z {type=MovieSceneScriptingFloatChannel, default=1}
       spine_02_ctrl.Location.X {type=MovieSceneScriptingFloatChannel, default=0}
       spine_02_ctrl.Location.Y {type=MovieSceneScriptingFloatChannel, default=0}
       spine_02_ctrl.Location.Z {type=MovieSceneScriptingFloatChannel, default=0}
       spine_02_ctrl.Rotation.X {type=MovieSceneScriptingFloatChannel, default=0}
       spine_02_ctrl.Rotation.Y {type=MovieSceneScriptingFloatChannel, default=0}
       spine_02_ctrl.Rotation.Z {type=MovieSceneScriptingFloatChannel, default=0}
       spine_02_ctrl.Scale.X {type=MovieSceneScriptingFloatChannel, default=1}
       spine_02_ctrl.Scale.Y {type=MovieSceneScriptingFloatChannel, default=1}
       spine_02_ctrl.Scale.Z {type=MovieSceneScriptingFloatChannel, default=1}
       spine_03_ctrl.Location.X {type=MovieSceneScriptingFloatChannel, default=0}
       spine_03_ctrl.Location.Y {type=MovieSceneScriptingFloatChannel, default=0}
       spine_03_ctrl.Location.Z {type=MovieSceneScriptingFloatChannel, default=0}
       spine_03_ctrl.Rotation.X {type=MovieSceneScriptingFloatChannel, default=0}
       spine_03_ctrl.Rotation.Y {type=MovieSceneScriptingFloatChannel, default=0}
       spine_03_ctrl.Rotation.Z {type=MovieSceneScriptingFloatChannel, default=0}
       spine_03_ctrl.Scale.X {type=MovieSceneScriptingFloatChannel, default=1}
       spine_03_ctrl.Scale.Y {type=MovieSceneScriptingFloatChannel, default=1}
       spine_03_ctrl.Scale.Z {type=MovieSceneScriptingFloatChannel, default=1}
       spine_04_ctrl.Location.X {type=MovieSceneScriptingFloatChannel, default=0}
       spine_04_ctrl.Location.Y {type=MovieSceneScriptingFloatChannel, default=0}
       spine_04_ctrl.Location.Z {type=MovieSceneScriptingFloatChannel, default=0}
       spine_04_ctrl.Rotation.X {type=MovieSceneScriptingFloatChannel, default=0}
       spine_04_ctrl.Rotation.Y {type=MovieSceneScriptingFloatChannel, default=0}
       spine_04_ctrl.Rotation.Z {type=MovieSceneScriptingFloatChannel, default=0}
       spine_04_ctrl.Scale.X {type=MovieSceneScriptingFloatChannel, default=1}
       spine_04_ctrl.Scale.Y {type=MovieSceneScriptingFloatChannel, default=1}
       spine_04_ctrl.Scale.Z {type=MovieSceneScriptingFloatChannel, default=1}
       head_ctrl.Location.X {type=MovieSceneScriptingFloatChannel, default=0}
       head_ctrl.Location.Y {type=MovieSceneScriptingFloatChannel, default=0}
       head_ctrl.Location.Z {type=MovieSceneScriptingFloatChannel, default=0}
       head_ctrl.Rotation.X {type=MovieSceneScriptingFloatChannel, default=0}
       head_ctrl.Rotation.Y {type=MovieSceneScriptingFloatChannel, default=0}
       head_ctrl.Rotation.Z {type=MovieSceneScriptingFloatChannel, default=0}
       head_ctrl.Scale.X {type=MovieSceneScriptingFloatChannel, default=1}
       head_ctrl.Scale.Y {type=MovieSceneScriptingFloatChannel, default=1}
       head_ctrl.Scale.Z {type=MovieSceneScriptingFloatChannel, default=1}
       clavicle_l_ctrl.Location.X {type=MovieSceneScriptingFloatChannel, default=0}
       clavicle_l_ctrl.Location.Y {type=MovieSceneScriptingFloatChannel, default=0}
       clavicle_l_ctrl.Location.Z {type=MovieSceneScriptingFloatChannel, default=0}
       clavicle_l_ctrl.Rotation.X {type=MovieSceneScriptingFloatChannel, default=0}
       clavicle_l_ctrl.Rotation.Y {type=MovieSceneScriptingFloatChannel, default=0}
       clavicle_l_ctrl.Rotation.Z {type=MovieSceneScriptingFloatChannel, default=0}
       clavicle_l_ctrl.Scale.X {type=MovieSceneScriptingFloatChannel, default=1}
       clavicle_l_ctrl.Scale.Y {type=MovieSceneScriptingFloatChannel, default=1}
       clavicle_l_ctrl.Scale.Z {type=MovieSceneScriptingFloatChannel, default=1}
       clavicle_r_ctrl.Location.X {type=MovieSceneScriptingFloatChannel, default=0}
       clavicle_r_ctrl.Location.Y {type=MovieSceneScriptingFloatChannel, default=0}
       clavicle_r_ctrl.Location.Z {type=MovieSceneScriptingFloatChannel, default=0}
       clavicle_r_ctrl.Rotation.X {type=MovieSceneScriptingFloatChannel, default=0}
       clavicle_r_ctrl.Rotation.Y {type=MovieSceneScriptingFloatChannel, default=0}
       clavicle_r_ctrl.Rotation.Z {type=MovieSceneScriptingFloatChannel, default=0}
       clavicle_r_ctrl.Scale.X {type=MovieSceneScriptingFloatChannel, default=1}
       clavicle_r_ctrl.Scale.Y {type=MovieSceneScriptingFloatChannel, default=1}
       clavicle_r_ctrl.Scale.Z {type=MovieSceneScriptingFloatChannel, default=1}
       neck_01_ctrl.Location.X {type=MovieSceneScriptingFloatChannel, default=0}
       neck_01_ctrl.Location.Y {type=MovieSceneScriptingFloatChannel, default=0}
       neck_01_ctrl.Location.Z {type=MovieSceneScriptingFloatChannel, default=0}
       neck_01_ctrl.Rotation.X {type=MovieSceneScriptingFloatChannel, default=0}
       neck_01_ctrl.Rotation.Y {type=MovieSceneScriptingFloatChannel, default=0}
       neck_01_ctrl.Rotation.Z {type=MovieSceneScriptingFloatChannel, default=0}
       neck_01_ctrl.Scale.X {type=MovieSceneScriptingFloatChannel, default=1}
       neck_01_ctrl.Scale.Y {type=MovieSceneScriptingFloatChannel, default=1}
       neck_01_ctrl.Scale.Z {type=MovieSceneScriptingFloatChannel, default=1}
       neck_02_ctrl.Location.X {type=MovieSceneScriptingFloatChannel, default=0}
       neck_02_ctrl.Location.Y {type=MovieSceneScriptingFloatChannel, default=0}
       neck_02_ctrl.Location.Z {type=MovieSceneScriptingFloatChannel, default=0}
       neck_02_ctrl.Rotation.X {type=MovieSceneScriptingFloatChannel, default=0}
       neck_02_ctrl.Rotation.Y {type=MovieSceneScriptingFloatChannel, default=0}
       neck_02_ctrl.Rotation.Z {type=MovieSceneScriptingFloatChannel, default=0}
       neck_02_ctrl.Scale.X {type=MovieSceneScriptingFloatChannel, default=1}
       neck_02_ctrl.Scale.Y {type=MovieSceneScriptingFloatChannel, default=1}
       neck_02_ctrl.Scale.Z {type=MovieSceneScriptingFloatChannel, default=1}
       head_ik_ctrl.Location.X {type=MovieSceneScriptingFloatChannel, default=0}
       head_ik_ctrl.Location.Y {type=MovieSceneScriptingFloatChannel, default=0}
       head_ik_ctrl.Location.Z {type=MovieSceneScriptingFloatChannel, default=0}
       head_ik_ctrl.Rotation.X {type=MovieSceneScriptingFloatChannel, default=0}
       head_ik_ctrl.Rotation.Y {type=MovieSceneScriptingFloatChannel, default=0}
       head_ik_ctrl.Rotation.Z {type=MovieSceneScriptingFloatChannel, default=0}
       head_ik_ctrl.Scale.X {type=MovieSceneScriptingFloatChannel, default=1}
       head_ik_ctrl.Scale.Y {type=MovieSceneScriptingFloatChannel, default=1}
       head_ik_ctrl.Scale.Z {type=MovieSceneScriptingFloatChannel, default=1}
       chest_ctrl.Location.X {type=MovieSceneScriptingFloatChannel, default=0}
       chest_ctrl.Location.Y {type=MovieSceneScriptingFloatChannel, default=0}
       chest_ctrl.Location.Z {type=MovieSceneScriptingFloatChannel, default=0}
       chest_ctrl.Rotation.X {type=MovieSceneScriptingFloatChannel, default=0}
       chest_ctrl.Rotation.Y {type=MovieSceneScriptingFloatChannel, default=0}
       chest_ctrl.Rotation.Z {type=MovieSceneScriptingFloatChannel, default=0}
       chest_ctrl.Scale.X {type=MovieSceneScriptingFloatChannel, default=1}
       chest_ctrl.Scale.Y {type=MovieSceneScriptingFloatChannel, default=1}
       chest_ctrl.Scale.Z {type=MovieSceneScriptingFloatChannel, default=1}
       chest_tan_ctrl.Location.X {type=MovieSceneScriptingFloatChannel, default=0}
       chest_tan_ctrl.Location.Y {type=MovieSceneScriptingFloatChannel, default=-0.0303}
       chest_tan_ctrl.Location.Z {type=MovieSceneScriptingFloatChannel, default=0.265}
       chest_tan_ctrl.Rotation.X {type=MovieSceneScriptingFloatChannel, default=0}
       chest_tan_ctrl.Rotation.Y {type=MovieSceneScriptingFloatChannel, default=0}
       chest_tan_ctrl.Rotation.Z {type=MovieSceneScriptingFloatChannel, default=0}
       chest_tan_ctrl.Scale.X {type=MovieSceneScriptingFloatChannel, default=1}
       chest_tan_ctrl.Scale.Y {type=MovieSceneScriptingFloatChannel, default=1}
       chest_tan_ctrl.Scale.Z {type=MovieSceneScriptingFloatChannel, default=1}
      ]
     }]
   }
   {
    type=MovieSceneControlRigParameterTrack
    display_name="Fortnite_Legacy_Face_CtrlRig (Layered)"
    sections [{
      type=MovieSceneControlRigParameterSection
      channels [
       Flipbook {type=MovieSceneScriptingBoolChannel, default=false}
       Face_Ctrl_Master.Location.X {type=MovieSceneScriptingFloatChannel, default=0}
       Face_Ctrl_Master.Location.Y {type=MovieSceneScriptingFloatChannel, default=0}
       Face_Ctrl_Master.Location.Z {type=MovieSceneScriptingFloatChannel, default=0}
       Face_Ctrl_Master.Rotation.X {type=MovieSceneScriptingFloatChannel, default=0}
       Face_Ctrl_Master.Rotation.Y {type=MovieSceneScriptingFloatChannel, default=0}
       Face_Ctrl_Master.Rotation.Z {type=MovieSceneScriptingFloatChannel, default=0}
       Face_Ctrl_Master.Scale.X {type=MovieSceneScriptingFloatChannel, default=1}
       Face_Ctrl_Master.Scale.Y {type=MovieSceneScriptingFloatChannel, default=1}
       Face_Ctrl_Master.Scale.Z {type=MovieSceneScriptingFloatChannel, default=1}
       gui_follow_head {type=MovieSceneScriptingFloatChannel, default=0}
      ]
     }]
   }
  ]
 }
 Body {id=#3, parent_id=#2, possessed_class=SkeletalMeshComponentBudgeted, tracks []}
 Face {
  id=#4
  display_name=Face Accessory
  parent_id=#2
  possessed_class=SkeletalMeshComponentBudgeted
  tracks []
 }
]
display_rate {denominator=1, numerator=30}
tick_resolution {denominator=1, numerator=24000}
tracks []
